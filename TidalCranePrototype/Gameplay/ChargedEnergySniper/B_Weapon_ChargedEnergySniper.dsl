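; Charged energy sniper weapon Blueprint that repurposes the overheat value handler to display charge-up visuals.
; While charging, a looping timer feeds it a "Heating" value blended from charge-to-minimum progress, plus charge percent once the minimum is reached.
; On charge down it stops that timer, signals full "Overheated", then a short one-shot timer sends zero as "Cooling".
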(event EventBeginPlay
  (|CallParentFunction) ; This node is disabled and will not be called.\nDrag off pins to build functionality.
)

(event EventTick
  (params (DeltaSeconds :type Float%20%28single-precision%29))
  (|CallParentFunction DeltaSeconds) ; This node is disabled and will not be called.\nDrag off pins to build functionality.
)

(event EventOnChargeUp
  (bind _returnvalue (Utilities|Time|SetTimerbyEvent (delegate Custom|ChargingUp) 0.1 true))
  (Variables|Default|SetChargeUpVisualsTimer _returnvalue))

(event Custom|ChargingUp
  (Overheat|HandleOverheatValueChanged (/ (Math|Float|SelectFloat (+ (FortWeapon|GetChargePercent) (/ (FortWeapon|GetChargetoMinPercent) 3.0)) (/ (FortWeapon|GetChargetoMinPercent) 3.0) (>= (FortWeapon|GetChargetoMinPercent) 1.0)) 2.0) 1.0 "Heating"))

(event EventOnChargeDown
  (Utilities|Time|ClearandInvalidateTimerbyHandle (Variables|Default|GetChargeUpVisualsTimer))
  (Overheat|HandleOverheatValueChanged 1.0 1.0 "Overheated")
  (Utilities|Time|SetTimerbyEvent (delegate Custom|ChargeDown) 0.9))

(event Custom|ChargeDown
  (Overheat|HandleOverheatValueChanged 0.0 1.0 "Cooling"))

(fn ConstructionScript
  (|CallParentFunction))
