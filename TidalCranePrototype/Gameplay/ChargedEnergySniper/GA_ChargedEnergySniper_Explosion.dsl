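; Gameplay ability for the charged energy sniper's pulse explosion, triggered by a gameplay event carrying charge as magnitude.
; A targeting trace selects a water, player, vault door or land impact cue, plays hit/miss force feedback and records the impact location.
; That location becomes the custom source transform for explosion targeting.
; The server asynchronously applies pawn and world damage effects with a per-tick target cap, with full charge changing the pawn effect.
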(event Ability|EventActivateAbilityFromEvent
  (params (EventData :type Gameplay%20Event%20Data%20Structure%20%28by%20ref%29))
  (bind (_eventtag _instigator _target _optionalobject _optionalobject2 _contexthandle _instigatortags _targettags _eventmagnitude) (Utilities|Struct|BreakGameplayEventData EventData))
  (bind _hotfix_maxtargetstoprocesspertick (Variables|Hotfix|GetHotfix_MaxTargetsToProcessPerTick))
  ;[0.01753 0.3906 0.1136 0.5] ActivateAbility: On the client, invoke the pulse trace sequence.
  (Variables|Default|SetChargeValue _eventmagnitude)
  (Ability|ExecuteGameplayCueOnOwner (Variables|Settings|GetPulseSpawnGameplayCueTag))
  (bind (_asynctaskproxy _targetdata) (Ability|Tasks|PerformTargetingAbilityTask "Targeting" "/ThermalLanceGameplay/Gameplay/Targeting/TP_ThermalLance_Targeting.TP_ThermalLance_Targeting")
    (:then
      (Variables|Transient|SetAsyncTask_Targeting _asynctaskproxy))
    (:Targeted
      (bind (_bsuccess _gameplaycue _hitresultforgameplaycue _impactpointforexplosiondamage) (Default|ParseTargetData _targetdata))
      (CallFunction|PlayLocalForceFeedback _bsuccess)
      (if _bsuccess
        (Ability|ExecuteGameplayCueWithParamsOnOwner _gameplaycue (CallFunction|MakeGameplayCueParamforPulseImpact _hitresultforgameplaycue))
        (Variables|Transient|SetImpactLocation _impactpointforexplosiondamage)
        (bind _weapon (Default|GetWeaponBP)
          (:Valid
            (Pulse|SetLatestPulseHitImpactLocation _weapon _impactpointforexplosiondamage)
            (Ability|HasAuthority
              (:True
                ;;[0.1364 0.05835 0.15 0.5] Async damage application on the server.
                (bind (_asynctaskproxy_2 _targetdata_2) (Ability|Tasks|PerformTargetingAbilityTask "None" (Variables|Settings|GetExplosionTargetingPresetPawn) true false true)
                  (:Targeted
                    _)
                  (:Cancelled
                    _))
                (Ability|Tasks|ApplyGameplayEffectToTargetDataAsync _targetdata_2 (Variables|Settings|GetExplosionPawnDamageGameplayEffect) (Math|Integer|SelectInt 1 0 (>= (Variables|Default|GetChargeValue) 1.0)) 1 _hotfix_maxtargetstoprocesspertick)
                (bind (_asynctaskproxy_3 _targetdata_3) (Ability|Tasks|PerformTargetingAbilityTask "None" (Variables|Settings|GetExplosionTargetingPresetWorld) true false true)
                  (:Targeted
                    _)
                  (:Cancelled
                    _))
                (Ability|Tasks|ApplyGameplayEffectToTargetDataAsync _targetdata_3 (Variables|Settings|GetExplosionWorldDamageGameplayEffect) 1 1 _hotfix_maxtargetstoprocesspertick))
              (:False
                _))
            (Ability|EndAbility))
          (:Invalid
            _))
        (else
          _))
      (Ability|CancelAbility))
    (:Cancelled
      _)))

(event Ability|EventOnEndAbility
  (params (bWasCancelled :type Boolean))
  ;[0.15 0.1021 0.0253 0.5] EndAbility: Clean up.
  (Utilities|FlowControl|Sequence
    (:then_0
      (bind _asynctask_targeting (Variables|Transient|GetAsyncTask_Targeting)
        (:then
          (GameplayTasks|EndTask _asynctask_targeting))
        (:else)))
    (:then_1
      (|CallParentFunction bWasCancelled))))

(fn GetCustomAbilitySourceTransform
  ; Used by the GameplayEffectContainer when targeting the pulse explosion.
  (return (Variables|Transient|GetImpactLocation) "0, 0, 0" "1.000000,1.000000,1.000000"))

(fn MakeGameplayCueParamForPulseImpact
  (params (HitResult :type Hit%20Result%20Structure%20%28by%20ref%29))
  (bind (_normalizedmagnitude _rawmagnitude _effectcontext _matchedtagname _originaltag _aggregatedsourcetags _aggregatedtargettags _location _normal _instigator _effectcauser _sourceobject _physicalmaterial _gameplayeffectlevel _abilitylevel _targetattachcomponent _breplicatelocationwhenusingminimalrepproxy) (Ability|GameplayCue|BreakGameplayCueParameters (GameplayCue|MakeGameplayCueParametersfromHitResult HitResult)))
  (bind _output_get (LocalVariables|SetEffectContext (Class|AthenaFunctionLibrary|MakeSimpleFortEffectContext (CallFunction|GetAvatarasFortPlayerPawn) (Ability|GetCurrentSourceWeapon))))
  (Ability|EffectContext|AddHitResult _output_get HitResult)
  (return (Ability|GameplayCue|MakeGameplayCueParameters _normalizedmagnitude _rawmagnitude _output_get _matchedtagname _originaltag _aggregatedsourcetags _aggregatedtargettags _location _normal _instigator _effectcauser _sourceobject _physicalmaterial _gameplayeffectlevel _abilitylevel _targetattachcomponent _breplicatelocationwhenusingminimalrepproxy)))

(fn ParseTargetData
  (params (TargetData :type Gameplay%20Ability%20Target%20Data%20Handle%20Structure%20%28by%20ref%29))
  (bind _returnvalue (Ability|TargetData|GetHitResultfromTargetData TargetData))
  (bind (_bblockinghit _binitialoverlap _time _distance _location _impactpoint _normal _impactnormal _physmat _hitactor _hitcomponent _hitbonename _bonename _hititem _elementindex _faceindex _tracestart _traceend) (Collision|BreakHitResult _returnvalue))
  (bind (_asfort_water_body_actor _bsuccess) (Utilities|Casting|CastToFortWaterBodyActor (LocalVariables|GetHitActor)))
  ; Comment
  (Utilities|FlowControl|Sequence
    (:then_0
      (if (Ability|TargetData|TargetDataHasHitResult TargetData)
        (LocalVariables|SetHitActor _hitactor)
        (LocalVariables|SetImpactPoint _location)
        (LocalVariables|SetHitResult (Collision|MakeHitResult true _binitialoverlap _time _distance _location _impactpoint _normal _impactnormal _physmat _hitactor _hitcomponent _hitbonename _bonename _hititem _elementindex _faceindex _tracestart _traceend))
        (if _bsuccess
          (return true (Variables|Settings|GetPulseExplodeWaterGameplayCueTag) (LocalVariables|GetHitResult) (LocalVariables|GetImpactPoint))
          (else
            (Utilities|Casting|CastToFortPlayerPawn (LocalVariables|GetHitActor)
              (:then
                (return true (Variables|Settings|GetPulseExplodePlayerGameplayCue) (LocalVariables|GetHitResult) (LocalVariables|GetImpactPoint)))
              (:CastFailed
                (bind _asbuilding_gameplay_actor (Utilities|Casting|CastToBuildingGameplayActor (LocalVariables|GetHitActor))
                  (:then
                    (if (GameplayTags|HasMatchingGameplayTag _asbuilding_gameplay_actor "(TagName=\"Athena.Quests.Building.Vault.Door\")")
                      (return true (Variables|Settings|GetPulseExplodeVaultGameplayCue) (LocalVariables|GetHitResult) (LocalVariables|GetImpactPoint))
                      (else
                        _)))
                  (:CastFailed
                    _))
                (return true (Variables|Settings|GetPulseExplodeLandGameplayCueTag) (LocalVariables|GetHitResult) (LocalVariables|GetImpactPoint))))))))
    (:then_1
      (return false 0 0 "0, 0, 0"))))

(fn PlayLocalForceFeedback
  (params (bSuccessPulseHit :type Boolean))
  ;[0.04024 0.15 0.1428 0.5] Play force feedback based on if the pulse successfully hit something or not.
  (bind _outfortplayerpawn (CallFunction|GetAvatarasValidFortPlayerPawn)
    (:ValidFortPlayerPawn
      (bind _returnvalue (Fortnite|GetFortPlayerControllerfromActor _outfortplayerpawn))
      (Utilities|IsValid _returnvalue
        (:Is%20Valid
          (Game|Feedback|LocalPlayForceFeedback _returnvalue (select bSuccessPulseHit (Variables|ForceFeedback|GetForceFeedbackPulseHitSuccess) (Variables|ForceFeedback|GetForceFeedbackPulseHitFail)) false "None"))))))
